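# Material function that stamps one frame from a flipbook atlas at an adjustable position, rotation and scale, defaulting to the CopperMoggy emote emoji texture.
# UVs are offset, rotated and center-scaled, then shifted to the cell chosen by Index using Fmod and Floor, with grid counts defaulting to 3x3.
# It outputs clamp-sampled RGB and Alpha, plus a separate Step-built Mask hard-coded to the middle third.
# A FlipBook function call is left unconnected.
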MaterialFunction /BRCosmetics/Animation/Props/CopperMoggy/Materials/MF_FlipbookStamp
graph {nodes [
  n1 = NamedRerouteDeclaration(None=FunctionInput() {bUsePreviewValueAsDefault=true, inputName=Index, inputType=FunctionInput_Scalar}) {name=Index, nodeColor=<rgba 0 0.2734 1 1>}
  n2 = NamedRerouteUsage()
  n3 = Divide(A=Subtract(A=Fmod(A=@n1, B=@n2)), B=NamedRerouteUsage()) {constB=3}
  n4 = Divide(A=@n1, B=NamedRerouteUsage()) {constB=3}
  n5 = Divide(A=Subtract(A=Floor(None=@n4)), B=NamedRerouteUsage()) {constB=3}
  n6 = TextureSample(
   UVs=Add(A=NamedRerouteUsage(), B=AppendVector(A=@n3, B=@n5))
   Tex=FunctionInput() {inputName=Texture, inputType=FunctionInput_Texture2D, sortPriority=3}
  ) {samplerSource=SSM_Clamp_WorldGroupSettings, texture=/BRCosmetics/Animation/Props/CopperMoggy/Textures/T_Emote_CopperMoggy_Emojis}
  FunctionOutput(None=@n6.RGB)
  n7 = NamedRerouteUsage()
  n8 = ComponentMask(None=@n7) {r=true}
  n9 = Divide() {constA=2, constB=3}
  n10 = Max(A=Step(Y=@n8, X=Divide() {constB=3}), B=OneMinus(None=Step(Y=@n8, X=@n9)))
  n11 = ComponentMask(None=@n7) {g=true}
  n12 = Divide() {constA=2, constB=3}
  n13 = Max(A=Step(Y=@n11, X=Divide() {constB=3}), B=OneMinus(None=Step(Y=@n11, X=@n12)))
  FunctionOutput(None=OneMinus(None=Max(A=@n10, B=@n13))) {outputName=Mask, sortPriority=2}
  FunctionOutput(None=@n6.A) {outputName=Alpha, sortPriority=1}
  MaterialFunctionCall() {materialFunction=&FlipBook}
  n14 = FunctionInput() {bUsePreviewValueAsDefault=true, inputName=Row Count, inputType=FunctionInput_Scalar, sortPriority=1, previewValue=<3 0 0 1>}
  NamedRerouteDeclaration(None=@n14) {name=RowCount, nodeColor=<rgba 0 0.8125 1 1>}
  n15 = FunctionInput() {bUsePreviewValueAsDefault=true, inputName=Column Count, inputType=FunctionInput_Scalar, sortPriority=2, previewValue=<3 0 0 1>}
  NamedRerouteDeclaration(None=@n15) {name=Column Count, nodeColor=<rgba 1 0.8672 0 1>}
  n16 = Subtract(
   A=FunctionInput(Preview=TextureCoordinate()) {bUsePreviewValueAsDefault=true, inputName=UVs, inputType=FunctionInput_Vector2, sortPriority=4}
   B=FunctionInput() {bUsePreviewValueAsDefault=true, inputName=Offset, inputType=FunctionInput_Vector2, sortPriority=5}
  )
  n17 = FunctionInput() {bUsePreviewValueAsDefault=true, inputName=Rotate, inputType=FunctionInput_Scalar, sortPriority=6}
  n18 = FunctionInput() {bUsePreviewValueAsDefault=true, inputName=Scale, inputType=FunctionInput_Vector2, sortPriority=7, previewValue=<1 1 0 1>}
  NamedRerouteDeclaration(None=MaterialFunctionCall(UVs=Rotator(Coordinate=@n16, Time=@n17) {speed=1}, Texture Scale=@n18) {materialFunction=&ScaleUVsByCenter}.UVs) {name=UVs, nodeColor=<rgba 0 1 0.8359 1>}
 ]}
