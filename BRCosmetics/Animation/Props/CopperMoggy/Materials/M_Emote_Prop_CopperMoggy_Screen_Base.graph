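# Translucent base material for the screens of the CopperMoggy emote prop, layering animated hacking effects over thin-film glass.
# Seven UV-bounded regions map a scrolling text texture, sharpened with screen derivatives.
# HackScreen_Animation expands object-centred sphere masks for a noisy wave and mosaic effect, alongside the DigitalMosaic function.
# Eight flipbook-stamped emojis with stepped rotational bob and damped-oscillation grow math blend into emissive on top.
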Material /BRCosmetics/Animation/Props/CopperMoggy/Materials/M_Emote_Prop_CopperMoggy_Screen_Base
graph {
 properties {
  bEnableResponsiveAA=true
  bOutputTranslucentVelocity=true
  bScreenSpaceReflections=true
  bUseMaterialAttributes=true
  bUsedWithClothing=true
  bUsedWithMeshDeformer=true
  bUsedWithMorphTargets=true
  bUsedWithNanite=true
  bUsedWithNiagaraMeshParticles=true
  bUsedWithSkeletalMesh=true
  blendMode=BLEND_Translucent
  ditherOpacityMask=true
  pixelDepthOffsetMode=PDOM_Legacy
  translucencyDirectionalLightingIntensity=5
  translucencyLightingMode=TLM_SurfacePerPixelLighting
  usedShadingModels=Default Lit
 }
 nodes [
  n1 = ScalarParameter() {group=CopperMoggy, parameterName=HackScreen_Animation}
  n2 = ScalarParameter() {defaultValue=3, group=CopperMoggy|Emojis, parameterName=EmojiGrow_Decay}
  n3 = Exponential!(None=Multiply(A=Multiply(A=@n1, B=@n2) {constB=3}) {constB=-1})
  n4 = Sine(None=Multiply(A=@n1, B=ScalarParameter() {defaultValue=15, group=CopperMoggy|Emojis, parameterName=EmojiGrow_Frequency}) {constB=15})
  Add(
   A=OneMinus(None=@n3)
   B=Multiply(A=Multiply(A=@n3, B=@n4), B=ScalarParameter() {defaultValue=1, group=CopperMoggy|Emojis, parameterName=EmojiGrow_Overshoot})
  )
  n5 = TextureSampleParameter2D() {
   group=00_Glass
   parameterName=TechArtMask
   samplerSource=SSM_Clamp_WorldGroupSettings
   samplerType=SAMPLERTYPE_Masks
   sortPriority=-96
   texture=&T_BlackMask
  }
  n6 = MaterialFunctionCall!(
   Material Attributes=MaterialFunctionCall!(TechArtMask=@n5.RGB) {materialFunction=/Game/Global/Materials/Character_Parent_2023/Glass_Parent_2023/MF_Glass}.Result
   Fresnel=Fresnel() {baseReflectFraction=0, exponent=1}
   Alpha=@n5.RGB
  ) {materialFunction=/Game/Global/MaterialFunctions/TechArt/MF_ThinFilm_TechArt}
  n7 = StaticSwitchParameter(
   True=MaterialFunctionCall!(In=@n6.Result) {materialFunction=/Game/Packages/Fortress_SharedMaterials/Base_Material_Functions/BaseCustomizationFunctions/MF_FN_Vehicle_CameraDither}.Result
   False=@n6.Result
  ) {group=00_Glass, parameterName=UseCameraDither, sortPriority=250}
  n8 = Multiply(A=Subtract(A=ObjectPositionWS(), B=WorldPosition().XYZ), B=ScalarParameter() {group=00_Glass, parameterName=VertexHide, sortPriority=301})
  n9 = GetMaterialAttributes(None=StaticSwitchParameter(
    True=SetMaterialAttributes!(MaterialAttributes=@n7, World Position Offset=@n8) {attributeSetTypes [#1]}
    False=MaterialFunctionCall!(Material=@n7) {materialFunction=/Game/Global/Materials/Character_Parent_2023/Functions/BaseFN_CharacterOpacity_2023}.Result
   ) {group=00_Glass, parameterName=UseVertexHide, sortPriority=300}) {attributeGetTypes [#2]}
  n10 = TextureCoordinate()
  n11 = Subtract(A=Step(Y=@n10, X=Constant2Vector() {g=0.65}), B=Step(Y=@n10, X=Constant2Vector() {g=1, r=0.65}))
  n12 = ComponentMask(None=@n11) {r=true}
  n13 = ComponentMask(None=@n11) {g=true}
  n14 = LinearInterpolate(
   A=@n9.Emissive Color
   B=Multiply(A=@n9.Emissive Color, B=VectorParameter() {parameterName=ScreenTint, defaultValue=<rgba 1 1 1 0>}.RGB)
   Alpha=Saturate(None=Multiply(A=@n12, B=@n13))
  )
  n15 = Max(A=NamedRerouteUsage(), B=NamedRerouteUsage())
  n16 = MaterialFunctionCall!(
   Material=SetMaterialAttributes!(MaterialAttributes=@n9.MaterialAttributes, Emissive Color=Add(A=@n14, B=@n15)) {attributeSetTypes [#2]}
   Tech Art Mask=NamedRerouteUsage()
  ) {materialFunction=/BRCosmetics/Characters/Player/Female/Medium/Bodies/F_MED_CopperMoggy_Swan/Skins/Prime/Materials/MF_CopperMoggy_DigitalMosaic}
  n17 = GetMaterialAttributes(None=@n16.Material Result) {attributeGetTypes [#2]}
  n18 = Multiply(
   A=Desaturation(None=NamedRerouteUsage(), Fraction=Constant() {r=-0.3})
   B=ScalarParameter() {defaultValue=1, group=CopperMoggy|Emojis, parameterName=EmojiIntensity, sortPriority=0}
  )
  n19 = Multiply(A=NamedRerouteUsage(), B=ScalarParameter() {defaultValue=1, group=CopperMoggy|Emojis, parameterName=EmojiFade, sortPriority=1})
  MaterialFunctionCall(In=SetMaterialAttributes(MaterialAttributes=@n17.MaterialAttributes, Emissive Color=LinearInterpolate(A=@n17.Emissive Color, B=@n18, Alpha=@n19)) {attributeSetTypes [#2]}) {materialFunction=&MF_TechArtOverrides}
  n20 = NamedRerouteUsage()
  Multiply(A=@n20) {constB=-1}
  Multiply(A=@n19, B=NamedRerouteUsage())
  n21 = ScalarParameter() {group=CopperMoggy, parameterName=HackScreen_Animation}
  n22 = Multiply(A=NamedRerouteUsage(), B=ScalarParameter() {parameterName=WaveNoiseAmount})
  n23 = SphereMask(
   A=WorldPosition().XYZ
   B=Add(A=ObjectPositionWS(), B=VectorParameter() {parameterName=WaveStartOffset, defaultValue=<rgba 0 0 0 1>}.RGB)
   Radius=Multiply(A=Subtract(A=@n21, B=@n22)) {constB=200}
   Hardness=ScalarParameter() {defaultValue=0.5, parameterName=WaveHardness}
  )
  n24 = Multiply(A=Multiply(A=Constant() {r=3}) {constB=4}, B=@n23)
  n25 = OneMinus(None=@n23)
  n26 = NamedRerouteUsage()
  n27 = Multiply(A=NamedRerouteUsage()) {constB=1.5}
  NamedRerouteDeclaration(None=Multiply(A=Multiply(A=@n24, B=@n25), B=Multiply(A=@n26, B=@n27))) {name=WaveEffect_Output, nodeColor=<rgba 0.2422 0 1 1>}
  NamedRerouteDeclaration(None=@n16.Grayscale Result) {name=GrayscaleMosaic, nodeColor=<rgba 0 1 0.7656 1>}
  n28 = VectorParameter() {
   group=CopperMoggy|Screen Advanced Params
   parameterName=Screen5_UVScaleOffset
   sortPriority=2
   channelNames {a=OffsetY, b=OffsetX, g=ScaleY, r=ScaleX}
   defaultValue=<rgba 3.5 3.5 -2.556 -2.3466>
  }
  n29 = Multiply(A=TextureCoordinate(), B=AppendVector(A=@n28.ScaleX, B=@n28.ScaleX))
  n30 = AppendVector(A=@n28.OffsetX, B=@n28.OffsetX)
  n31 = TextureCoordinate()
  n32 = VectorParameter() {
   group=CopperMoggy|Screen Advanced Params
   parameterName=Screen5_MaskBounds
   sortPriority=2
   channelNames {a=Y-Max, b=X-Max, g=Y-Min, r=X-Min}
   defaultValue=<rgba 0.3705 0.774 0.452 0.8735>
  }
  n33 = Subtract(A=Step(Y=@n31, X=AppendVector(A=@n32.X-Min, B=@n32.X-Min)), B=Step(Y=@n31, X=AppendVector(A=@n32.X-Max, B=@n32.X-Max)))
  n34 = Saturate(None=Multiply(A=ComponentMask(None=@n33) {r=true}, B=ComponentMask(None=@n33) {g=true}))
  NamedRerouteDeclaration(None=Multiply(A=Add(A=@n29, B=@n30), B=NamedRerouteDeclaration(None=@n34) {name=Screen5_Mask, nodeColor=<rgba 0 1 0.1328 1>})) {name=Screen5_UVs, nodeColor=<rgba 0.3125 0 1 1>}
  n35 = VectorParameter() {
   group=CopperMoggy|Screen Advanced Params
   parameterName=Screen6_UVScaleOffset
   sortPriority=2
   channelNames {a=OffsetY, b=OffsetX, g=ScaleY, r=ScaleX}
   defaultValue=<rgba 3.5 3.5 -1.3809 -2.2>
  }
  n36 = Multiply(A=TextureCoordinate(), B=AppendVector(A=@n35.ScaleX, B=@n35.ScaleX))
  n37 = AppendVector(A=@n35.OffsetX, B=@n35.OffsetX)
  n38 = TextureCoordinate()
  n39 = VectorParameter() {
   group=CopperMoggy|Screen Advanced Params
   parameterName=Screen6_MaskBounds
   sortPriority=2
   channelNames {a=Y-Max, b=X-Max, g=Y-Min, r=X-Min}
   defaultValue=<rgba 0.2345 0.83 0.316 0.9135>
  }
  n40 = Subtract(A=Step(Y=@n38, X=AppendVector(A=@n39.X-Min, B=@n39.X-Min)), B=Step(Y=@n38, X=AppendVector(A=@n39.X-Max, B=@n39.X-Max)))
  n41 = Saturate(None=Multiply(A=ComponentMask(None=@n40) {r=true}, B=ComponentMask(None=@n40) {g=true}))
  NamedRerouteDeclaration(None=Multiply(A=Add(A=@n36, B=@n37), B=NamedRerouteDeclaration(None=@n41) {name=Screen6_Mask, nodeColor=<rgba 0 1 0.1328 1>})) {name=Screen6_UVs, nodeColor=<rgba 0.3125 0 1 1>}
  n42 = VectorParameter() {
   group=CopperMoggy|Screen Advanced Params
   parameterName=Screen7_UVScaleOffset
   sortPriority=2
   channelNames {a=OffsetY, b=OffsetX, g=ScaleY, r=ScaleX}
   defaultValue=<rgba 3.5 3.5 -3.5388 -2.664>
  }
  n43 = Multiply(A=TextureCoordinate(), B=AppendVector(A=@n42.ScaleX, B=@n42.ScaleX))
  n44 = AppendVector(A=@n42.OffsetX, B=@n42.OffsetX)
  n45 = TextureCoordinate()
  n46 = VectorParameter() {
   group=CopperMoggy|Screen Advanced Params
   parameterName=Screen7_MaskBounds
   sortPriority=2
   channelNames {a=Y-Max, b=X-Max, g=Y-Min, r=X-Min}
   defaultValue=<rgba 0.4985 0.846 0.617 0.9535>
  }
  n47 = Subtract(A=Step(Y=@n45, X=AppendVector(A=@n46.X-Min, B=@n46.X-Min)), B=Step(Y=@n45, X=AppendVector(A=@n46.X-Max, B=@n46.X-Max)))
  n48 = Saturate(None=Multiply(A=ComponentMask(None=@n47) {r=true}, B=ComponentMask(None=@n47) {g=true}))
  NamedRerouteDeclaration(None=Multiply(A=Add(A=@n43, B=@n44), B=NamedRerouteDeclaration(None=@n48) {name=Screen7_Mask, nodeColor=<rgba 0 1 0.1328 1>})) {name=Screen7_UVs, nodeColor=<rgba 0.3125 0 1 1>}
  n49 = Max(A=NamedRerouteUsage(), B=NamedRerouteUsage())
  n50 = NamedRerouteUsage()
  n51 = Max(A=Max(A=Max(A=@n49, B=@n50), B=NamedRerouteUsage()), B=NamedRerouteUsage())
  n52 = NamedRerouteUsage()
  NamedRerouteDeclaration(None=Max(A=Max(A=@n51, B=@n52), B=NamedRerouteUsage())) {name=TextScreen_UVs, nodeColor=<rgba 0 1 0.25 1>}
  n53 = TextureSampleParameter2D(UVs=NamedRerouteUsage()) {
   parameterName=TextTexture
   samplerSource=SSM_Wrap_WorldGroupSettings
   samplerType=SAMPLERTYPE_Alpha
   texture=/BRCosmetics/Animation/Props/CopperMoggy/Textures/T_Emote_CopperMoggy_Text
  }
  n54 = Add(A=Abs(None=DDX(Value=@n53.RGB)), B=Abs(None=DDY(Value=@n53.RGB)))
  n55 = Add(A=Divide(A=Subtract(A=@n53.RGB) {constB=0.5}, B=Max(A=@n54) {constB=0})) {constB=0.5}
  NamedRerouteDeclaration(None=Multiply(A=Saturate(None=@n55), B=VectorParameter() {group=CopperMoggy, parameterName=TextColor, defaultValue=<rgba 0.0646 1.2 0.7874 1>}.RGB)) {name=Output, nodeColor=<rgba 1 0 0.3984 1>}
  n56 = Max(A=NamedRerouteUsage(), B=NamedRerouteUsage())
  n57 = NamedRerouteUsage()
  n58 = Max(A=Max(A=Max(A=@n56, B=@n57), B=NamedRerouteUsage()), B=NamedRerouteUsage())
  n59 = NamedRerouteUsage()
  NamedRerouteDeclaration(None=Max(A=Max(A=@n58, B=@n59), B=NamedRerouteUsage())) {name=TextScreen_Masks, nodeColor=<rgba 0.7344 0 1 1>}
  n60 = NamedRerouteUsage()
  n61 = AppendVector(A=ComponentMask(None=@n60) {r=true}, B=Add(A=ComponentMask(None=@n60) {g=true}, B=NamedRerouteUsage()))
  n62 = NamedRerouteUsage()
  n63 = NamedRerouteUsage()
  n64 = AppendVector(A=Add(A=ComponentMask(None=@n63) {r=true}, B=@n20), B=ComponentMask(None=@n63) {g=true})
  n65 = NamedRerouteUsage()
  NamedRerouteDeclaration(None=Max(A=Multiply(A=@n61, B=@n62), B=Multiply(A=@n64, B=@n65))) {name=AnimatedTextScreen_UVs, nodeColor=<rgba 0.9688 0 1 1>}
  n66 = VectorParameter() {
   group=CopperMoggy|Screen Advanced Params
   parameterName=Screen1_UVScaleOffset
   sortPriority=2
   channelNames {a=OffsetY, b=OffsetX, g=ScaleY, r=ScaleX}
   defaultValue=<rgba 3.8 3.8 -0.152 -2.6299>
  }
  n67 = Multiply(A=TextureCoordinate(), B=AppendVector(A=@n66.ScaleX, B=@n66.ScaleX))
  n68 = AppendVector(A=@n66.OffsetX, B=@n66.OffsetX)
  n69 = TextureCoordinate()
  n70 = VectorParameter() {
   group=CopperMoggy|Screen Advanced Params
   parameterName=Screen1_MaskBounds
   sortPriority=2
   channelNames {a=Y-Max, b=X-Max, g=Y-Min, r=X-Min}
   defaultValue=<rgba 0 0.75 0.2 0.87>
  }
  n71 = Subtract(A=Step(Y=@n69, X=AppendVector(A=@n70.X-Min, B=@n70.X-Min)), B=Step(Y=@n69, X=AppendVector(A=@n70.X-Max, B=@n70.X-Max)))
  n72 = Saturate(None=Multiply(A=ComponentMask(None=@n71) {r=true}, B=ComponentMask(None=@n71) {g=true}))
  NamedRerouteDeclaration(None=Multiply(A=Add(A=@n67, B=@n68), B=NamedRerouteDeclaration(None=@n72) {name=Screen1_Mask, nodeColor=<rgba 1 0 0.0234 1>})) {name=Screen1_UVs, nodeColor=<rgba 0.3125 0 1 1>}
  NamedRerouteDeclaration(None=Multiply(A=Time(), B=ScalarParameter() {defaultValue=0.75, parameterName=TextScreen_Speed})) {name=TextSpeed, nodeColor=<rgba 0 0.8828 1 1>}
  n73 = Max(A=Max(A=NamedRerouteUsage(), B=NamedRerouteUsage()), B=NamedRerouteUsage())
  n74 = NamedRerouteUsage()
  NamedRerouteDeclaration(None=Max(A=Max(A=@n73, B=@n74), B=NamedRerouteUsage())) {name=TextScreen_Vertical_Masks, nodeColor=<rgba 0 1 0.3672 1>}
  NamedRerouteDeclaration(None=Max(A=NamedRerouteUsage(), B=NamedRerouteUsage())) {name=TextScreen_Horizontal_Masks, nodeColor=<rgba 1 0.0469 0 1>}
  NamedRerouteDeclaration(None=@n16.GridMask) {name=GridMask, nodeColor=<rgba 1 0 0.4922 1>}
  NamedRerouteDeclaration(None=@n16.Color Result) {name=ColorMosaic, nodeColor=<rgba 1 0 0.1875 1>}
  n75 = ObjectPositionWS()
  n76 = VectorParameter() {parameterName=WaveStartOffset, defaultValue=<rgba 0 0 0 1>}
  n77 = Multiply(A=ScalarParameter() {group=CopperMoggy, parameterName=HackScreen_Animation}) {constB=200}
  NamedRerouteDeclaration(None=SphereMask(A=WorldPosition().XYZ, B=Add(A=@n75, B=@n76.RGB), Radius=Divide(A=@n77) {constB=1.5}) {hardnessPercent=50}) {name=MosaicEffectMask, nodeColor=<rgba 1 0.5859 0 1>}
  NamedRerouteDeclaration(None=@n16.NoiseResult) {name=MosaicNoise, nodeColor=<rgba 0 1 0.0391 1>}
  NamedRerouteDeclaration(None=TextureObject() {texture=/BRCosmetics/Animation/Props/CopperMoggy/Textures/T_Emote_CopperMoggy_Emojis}) {name=EmojiTexture, nodeColor=<rgba 0 0.2969 1 1>}
  n78 = Multiply(A=Time(), B=ScalarParameter() {defaultValue=2, group=CopperMoggy|Emojis, parameterName=EmojiBobSpeed})
  NamedRerouteDeclaration(None=Fmod(A=Floor(None=@n78), B=Constant() {r=2})) {name=EmojiBob_Animation, nodeColor=<rgba 0.8281 1 0 1>}
  NamedRerouteDeclaration(None=ScalarParameter() {group=CopperMoggy|Emojis, parameterName=EmojiBobAmount}) {name=EmojiBob_Amount, nodeColor=<rgba 0.8516 0 1 1>}
  n79 = VectorParameter() {
   group=CopperMoggy|Screen Advanced Params
   parameterName=Screen2_UVScaleOffset
   sortPriority=2
   channelNames {a=OffsetY, b=OffsetX, g=ScaleY, r=ScaleX}
   defaultValue=<rgba 3.8 3.8 -0.152 -3.35>
  }
  n80 = Multiply(A=TextureCoordinate(), B=AppendVector(A=@n79.ScaleX, B=@n79.ScaleX))
  n81 = AppendVector(A=@n79.OffsetX, B=@n79.OffsetX)
  n82 = TextureCoordinate()
  n83 = VectorParameter() {
   group=CopperMoggy|Screen Advanced Params
   parameterName=Screen2_MaskBounds
   sortPriority=2
   channelNames {a=Y-Max, b=X-Max, g=Y-Min, r=X-Min}
   defaultValue=<rgba 0 0.92 0.2 0.96>
  }
  n84 = Subtract(A=Step(Y=@n82, X=AppendVector(A=@n83.X-Min, B=@n83.X-Min)), B=Step(Y=@n82, X=AppendVector(A=@n83.X-Max, B=@n83.X-Max)))
  n85 = Saturate(None=Multiply(A=ComponentMask(None=@n84) {r=true}, B=ComponentMask(None=@n84) {g=true}))
  NamedRerouteDeclaration(None=Multiply(A=Add(A=@n80, B=@n81), B=NamedRerouteDeclaration(None=@n85) {name=Screen2_Mask, nodeColor=<rgba 0 0.3437 1 1>})) {name=Screen2_UVs, nodeColor=<rgba 0.3125 0 1 1>}
  n86 = NamedRerouteUsage()
  n87 = Multiply(A=@n86, B=ScalarParameter() {defaultValue=3, group=CopperMoggy|Emojis, parameterName=EmojiGrow_Decay}) {constB=3}
  n88 = ScalarParameter() {defaultValue=15, group=CopperMoggy|Emojis, parameterName=EmojiGrow_Frequency}
  n89 = Multiply(A=Exponential(None=Multiply(A=@n87) {constB=-1}), B=Cosine(None=Multiply(A=@n86, B=@n88) {constB=15}))
  n90 = ScalarParameter() {defaultValue=1, group=CopperMoggy|Emojis, parameterName=EmojiGrow_Overshoot}
  NamedRerouteDeclaration(None=OneMinus(None=Multiply(A=@n89, B=@n90))) {name=EmojiGrow_Animation, nodeColor=<rgba 1 0.2109 0 1>}
  n91 = VectorParameter() {
   group=CopperMoggy|Emojis
   parameterName=Emoji1_OffsetRotateScale
   channelNames {a=Scale, b=Rotate, g=OffsetY, r=OffsetX}
   defaultValue=<rgba 0 0 0 1>
  }
  n92 = NamedRerouteUsage()
  n93 = Multiply(A=@n92) {constB=-1}
  n94 = NamedRerouteUsage()
  n95 = MaterialFunctionCall!(
   Index=Constant() {r=5}
   Row Count
   Column Count
   Texture=NamedRerouteUsage()
   UVs
   Offset=AppendVector(A=@n91.OffsetX, B=@n91.OffsetX)
   Rotate=Add(A=@n91.Rotate, B=LinearInterpolate(A=@n93, B=@n92, Alpha=@n94))
   Scale=LinearInterpolate(B=@n91.Scale, Alpha=NamedRerouteUsage())
  ) {materialFunction=&/MF_FlipbookStamp}
  n96 = Multiply(A=@n95.Alpha, B=@n95.Alpha)
  n97 = VectorParameter() {
   group=CopperMoggy|Emojis
   parameterName=Emoji2_OffsetRotateScale
   channelNames {a=Scale, b=Rotate, g=OffsetY, r=OffsetX}
   defaultValue=<rgba 0 0 0 1>
  }
  n98 = NamedRerouteUsage()
  n99 = Multiply(A=@n98) {constB=-1}
  n100 = NamedRerouteUsage()
  n101 = MaterialFunctionCall(
   Index=Constant() {r=1}
   Texture=NamedRerouteUsage()
   Offset=AppendVector(A=@n97.OffsetX, B=@n97.OffsetX)
   Rotate=Add(A=@n97.Rotate, B=LinearInterpolate(A=@n99, B=@n98, Alpha=@n100))
   Scale=LinearInterpolate(B=@n97.Scale, Alpha=NamedRerouteUsage())
  ) {materialFunction=&MF_FlipbookStamp}
  n102 = Multiply(A=@n101.Alpha, B=@n101.Alpha)
  n103 = VectorParameter() {
   group=CopperMoggy|Emojis
   parameterName=Emoji3_OffsetRotateScale
   channelNames {a=Scale, b=Rotate, g=OffsetY, r=OffsetX}
   defaultValue=<rgba 0 0 0 1>
  }
  n104 = NamedRerouteUsage()
  n105 = Multiply(A=@n104) {constB=-1}
  n106 = NamedRerouteUsage()
  n107 = MaterialFunctionCall(
   Index=Constant() {r=4}
   Texture=NamedRerouteUsage()
   Offset=AppendVector(A=@n103.OffsetX, B=@n103.OffsetX)
   Rotate=Add(A=@n103.Rotate, B=LinearInterpolate(A=@n105, B=@n104, Alpha=@n106))
   Scale=LinearInterpolate(B=@n103.Scale, Alpha=NamedRerouteUsage())
  ) {materialFunction=&MF_FlipbookStamp}
  n108 = Multiply(A=@n107.Alpha, B=@n107.Alpha)
  n109 = Max(A=Max(A=Multiply(A=@n95.Result, B=@n96), B=Multiply(A=@n101.Result, B=@n102)), B=Multiply(A=@n107.Result, B=@n108))
  n110 = VectorParameter() {
   group=CopperMoggy|Emojis
   parameterName=Emoji4_OffsetRotateScale
   channelNames {a=Scale, b=Rotate, g=OffsetY, r=OffsetX}
   defaultValue=<rgba 0 0 0 1>
  }
  n111 = NamedRerouteUsage()
  n112 = Multiply(A=@n111) {constB=-1}
  n113 = NamedRerouteUsage()
  n114 = MaterialFunctionCall(
   Index=Constant() {r=2}
   Texture=NamedRerouteUsage()
   Offset=AppendVector(A=@n110.OffsetX, B=@n110.OffsetX)
   Rotate=Add(A=@n110.Rotate, B=LinearInterpolate(A=@n112, B=@n111, Alpha=@n113))
   Scale=LinearInterpolate(B=@n110.Scale, Alpha=NamedRerouteUsage())
  ) {materialFunction=&MF_FlipbookStamp}
  n115 = Multiply(A=@n114.Alpha, B=@n114.Alpha)
  n116 = Multiply(A=@n114.Result, B=@n115)
  n117 = VectorParameter() {
   group=CopperMoggy|Emojis
   parameterName=Emoji5_OffsetRotateScale
   channelNames {a=Scale, b=Rotate, g=OffsetY, r=OffsetX}
   defaultValue=<rgba 0 0 0 1>
  }
  n118 = NamedRerouteUsage()
  n119 = Multiply(A=@n118) {constB=-1}
  n120 = NamedRerouteUsage()
  n121 = MaterialFunctionCall(
   Index=Constant() {r=6}
   Texture=NamedRerouteUsage()
   Offset=AppendVector(A=@n117.OffsetX, B=@n117.OffsetX)
   Rotate=Add(A=@n117.Rotate, B=LinearInterpolate(A=@n119, B=@n118, Alpha=@n120))
   Scale=LinearInterpolate(B=@n117.Scale, Alpha=NamedRerouteUsage())
  ) {materialFunction=&MF_FlipbookStamp}
  n122 = Multiply(A=@n121.Alpha, B=@n121.Alpha)
  n123 = VectorParameter() {
   group=CopperMoggy|Emojis
   parameterName=Emoji6_OffsetRotateScale
   channelNames {a=Scale, b=Rotate, g=OffsetY, r=OffsetX}
   defaultValue=<rgba 0 0 0 1>
  }
  n124 = NamedRerouteUsage()
  n125 = Multiply(A=@n124) {constB=-1}
  n126 = NamedRerouteUsage()
  n127 = MaterialFunctionCall(
   Index=Constant() {r=7}
   Texture=NamedRerouteUsage()
   Offset=AppendVector(A=@n123.OffsetX, B=@n123.OffsetX)
   Rotate=Add(A=@n123.Rotate, B=LinearInterpolate(A=@n125, B=@n124, Alpha=@n126))
   Scale=LinearInterpolate(B=@n123.Scale, Alpha=NamedRerouteUsage())
  ) {materialFunction=&MF_FlipbookStamp}
  n128 = Multiply(A=@n127.Alpha, B=@n127.Alpha)
  n129 = Max(A=Max(A=Max(A=@n109, B=@n116), B=Multiply(A=@n121.Result, B=@n122)), B=Multiply(A=@n127.Result, B=@n128))
  n130 = VectorParameter() {
   group=CopperMoggy|Emojis
   parameterName=Emoji7_OffsetRotateScale
   channelNames {a=Scale, b=Rotate, g=OffsetY, r=OffsetX}
   defaultValue=<rgba 0 0 0 1>
  }
  n131 = NamedRerouteUsage()
  n132 = Multiply(A=@n131) {constB=-1}
  n133 = NamedRerouteUsage()
  n134 = MaterialFunctionCall(
   Index=Constant() {r=8}
   Texture=NamedRerouteUsage()
   Offset=AppendVector(A=@n130.OffsetX, B=@n130.OffsetX)
   Rotate=Add(A=@n130.Rotate, B=LinearInterpolate(A=@n132, B=@n131, Alpha=@n133))
   Scale=LinearInterpolate(B=@n130.Scale, Alpha=NamedRerouteUsage())
  ) {materialFunction=&MF_FlipbookStamp}
  n135 = Multiply(A=@n134.Alpha, B=@n134.Alpha)
  n136 = Multiply(A=@n134.Result, B=@n135)
  n137 = VectorParameter() {
   group=CopperMoggy|Emojis
   parameterName=Emoji8_OffsetRotateScale
   channelNames {a=Scale, b=Rotate, g=OffsetY, r=OffsetX}
   defaultValue=<rgba 0 0 0 1>
  }
  n138 = NamedRerouteUsage()
  n139 = Multiply(A=@n138) {constB=-1}
  n140 = NamedRerouteUsage()
  n141 = MaterialFunctionCall(
   Index=Constant()
   Texture=NamedRerouteUsage()
   Offset=AppendVector(A=@n137.OffsetX, B=@n137.OffsetX)
   Rotate=Add(A=@n137.Rotate, B=LinearInterpolate(A=@n139, B=@n138, Alpha=@n140))
   Scale=LinearInterpolate(B=@n137.Scale, Alpha=NamedRerouteUsage())
  ) {materialFunction=&MF_FlipbookStamp}
  n142 = Multiply(A=@n141.Alpha, B=@n141.Alpha)
  NamedRerouteDeclaration(None=Max(A=Max(A=@n129, B=@n136), B=Multiply(A=@n141.Result, B=@n142))) {name=EmojisColor, nodeColor=<rgba 0 1 0.7656 1>}
  n143 = Max(A=Max(A=@n96, B=@n102), B=@n108)
  n144 = Max(A=Max(A=Max(A=@n143, B=@n115), B=@n122), B=@n128)
  NamedRerouteDeclaration(None=Max(A=Max(A=@n144, B=@n135), B=@n142)) {name=EmojiMask, nodeColor=<rgba 1 0.4453 0 1>}
  n145 = VectorParameter() {
   group=CopperMoggy|Screen Advanced Params
   parameterName=Screen3_UVScaleOffset
   sortPriority=2
   channelNames {a=OffsetY, b=OffsetX, g=ScaleY, r=ScaleX}
   defaultValue=<rgba 3.8 3.8 -2.6024 -2.7618>
  }
  n146 = Multiply(A=TextureCoordinate(), B=AppendVector(A=@n145.ScaleX, B=@n145.ScaleX))
  n147 = AppendVector(A=@n145.OffsetX, B=@n145.OffsetX)
  n148 = TextureCoordinate()
  n149 = VectorParameter() {
   group=CopperMoggy|Screen Advanced Params
   parameterName=Screen3_MaskBounds
   sortPriority=2
   channelNames {a=Y-Max, b=X-Max, g=Y-Min, r=X-Min}
   defaultValue=<rgba 0.368 0.91 0.45 0.982>
  }
  n150 = Subtract(A=Step(Y=@n148, X=AppendVector(A=@n149.X-Min, B=@n149.X-Min)), B=Step(Y=@n148, X=AppendVector(A=@n149.X-Max, B=@n149.X-Max)))
  n151 = Saturate(None=Multiply(A=ComponentMask(None=@n150) {r=true}, B=ComponentMask(None=@n150) {g=true}))
  NamedRerouteDeclaration(None=Multiply(A=Add(A=@n146, B=@n147), B=NamedRerouteDeclaration(None=@n151) {name=Screen3_Mask, nodeColor=<rgba 0.6641 1 0 1>})) {name=Screen3_UVs, nodeColor=<rgba 0.3125 0 1 1>}
  n152 = VectorParameter() {
   group=CopperMoggy|Screen Advanced Params
   parameterName=Screen4_UVScaleOffset
   sortPriority=2
   channelNames {a=OffsetY, b=OffsetX, g=ScaleY, r=ScaleX}
   defaultValue=<rgba 3.5 3.5 -1.8286 -2.4074>
  }
  n153 = Multiply(A=TextureCoordinate(), B=AppendVector(A=@n152.ScaleX, B=@n152.ScaleX))
  n154 = AppendVector(A=@n152.OffsetX, B=@n152.OffsetX)
  n155 = TextureCoordinate()
  n156 = VectorParameter() {
   group=CopperMoggy|Screen Advanced Params
   parameterName=Screen4_MaskBounds
   sortPriority=2
   channelNames {a=Y-Max, b=X-Max, g=Y-Min, r=X-Min}
   defaultValue=<rgba 0.216 0.67 0.44 0.725>
  }
  n157 = Subtract(A=Step(Y=@n155, X=AppendVector(A=@n156.X-Min, B=@n156.X-Min)), B=Step(Y=@n155, X=AppendVector(A=@n156.X-Max, B=@n156.X-Max)))
  n158 = Saturate(None=Multiply(A=ComponentMask(None=@n157) {r=true}, B=ComponentMask(None=@n157) {g=true}))
  NamedRerouteDeclaration(None=Multiply(A=Add(A=@n153, B=@n154), B=NamedRerouteDeclaration(None=@n158) {name=Screen4_Mask, nodeColor=<rgba 0 1 0.1328 1>})) {name=Screen4_UVs, nodeColor=<rgba 0.3125 0 1 1>}
  ScalarParameter() {parameterName=DEBUG_Animation}
  ScalarParameter() {group=CopperMoggy, parameterName=HackScreen_Animation}
 ]
}
